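# Physics asset for the Lux vehicle body skeletal mesh in the vehicle cosmetics content.
# It lists two bodies, one with no bone name and one on the root bone, each with no shapes shown.
# It defines no constraints, so no joints link the bodies.
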PhysicsAsset /VehicleCosmetics/Bodies/Lux/Meshes/SK_Body_Lux_PhysicsAsset
bodies {None [], root []}
constraints []
